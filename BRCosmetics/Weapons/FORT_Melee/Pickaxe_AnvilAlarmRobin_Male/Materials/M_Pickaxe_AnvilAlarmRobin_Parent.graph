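# Parent character-style material for the AnvilAlarmRobin male pickaxe, blending a reactive color and shine look over the standard CharacterParent 2023 PreFX/PostFX base.
# Bounding-box sphere and height masks drive a color-curve gradient on the upper part and a weakpoint tint on the handle.
# Fake specular from camera, sun and skylight reflection feeds emissive.
# A scalar Reactivity parameter gates the blend; timestamp-driven two-hit ramps are computed into a separate Reactivity reroute.
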Material /BRCosmetics/Weapons/FORT_Melee/Pickaxe_AnvilAlarmRobin_Male/Materials/M_Pickaxe_AnvilAlarmRobin_Parent
graph {
 properties {
  bAutomaticallySetUsageInEditor=false
  bUseMaterialAttributes=true
  bUsedWithMeshDeformer=true
  bUsedWithMorphTargets=true
  bUsedWithSkeletalMesh=true
  pixelDepthOffsetMode=PDOM_Legacy
  shadingModel=MSM_FromMaterialExpression
  usedShadingModels=Default Lit | Subsurface | Clear Coat | Subsurface Profile | Cloth
 }
 nodes [
  n1 = CameraVectorWS()
  Add(A=@n1, B=ScalarParameter() {defaultValue=1, parameterName=Param_4})
  n2 = AtmosphericLightVector()
  DotProduct(A=@n2, B=@n1)
  n3 = MaterialFunctionCall() {materialFunction=&MF_CharacterParent_2023_PreFX}
  n4 = GetMaterialAttributes(None=@n3.Result) {attributeGetTypes [#1]}
  n5 = MaterialFunctionCall() {materialFunction=&BoundingBoxBased_0-1_UVW}
  n6 = VectorParameter() {defaultValue=<rgba 0.5 0.5 0.8434 0>}
  n7 = ScalarParameter() {defaultValue=0.4909, parameterName=Radius}
  n8 = ScalarParameter() {parameterName=Hardness}
  n9 = Step(Y=ScalarParameter() {defaultValue=0.744}, X=@n5.B)
  n10 = Multiply(
   A=Multiply(A=SphereMask(A=@n5.RGB, B=@n6.RGB, Radius=@n7, Hardness=@n8), B=@n9)
   B=ScalarParameter() {defaultValue=3, parameterName=Mask Boost}
  )
  n11 = TextureSampleParameter2D(UVs=Saturate(None=@n10)) {
   parameterName=Color Curve
   samplerType=SAMPLERTYPE_LinearColor
   texture=/Game/Weapons/FORT_Melee/Pickaxe_AnvilAlarmRobin_Male/Materials/CA_AnvilAlarmRobin
  }
  n12 = ScalarParameter() {defaultValue=1.5, parameterName=Param_3}
  n13 = Multiply(A=@n11.RGB, B=@n12)
  n14 = DotProduct(A=PixelNormalWS(), B=CameraVectorWS())
  n15 = ScalarParameter() {defaultValue=0.5181, parameterName=Offset}
  n16 = Power(Base=Sine(None=Add(A=@n14, B=@n15)), Exp=ScalarParameter() {defaultValue=1.3337, parameterName="DArk  Smoothnes"})
  n17 = VectorParameter() {parameterName=Weakpoint Base Color, defaultValue=<rgba 0.4342 0.0065 1 1>}
  n18 = ScalarParameter() {defaultValue=0.5899, parameterName=Threshold}
  n19 = SmoothStep(Min=@n18, Max=Add(A=@n18, B=ScalarParameter() {defaultValue=0.0917, parameterName=Smoothness}), Value=@n5.B)
  n20 = NamedRerouteDeclaration(None=OneMinus(None=@n9)) {name=Handle, nodeColor=<rgba 0.3594 1 0 1>}
  n21 = LinearInterpolate(A=@n19, B=@n10, Alpha=@n9)
  n22 = LinearInterpolate(
   A=@n4.Base Color
   B=LinearInterpolate(A=LinearInterpolate(A=@n13, B=@n11.RGB, Alpha=@n16), B=Multiply(A=@n17.RGB, B=@n12) {constB=2}, Alpha=Multiply(A=@n19, B=@n20))
   Alpha=@n21
  )
  n23 = PixelNormalWS()
  n24 = LinearInterpolate(
   A=Power(Base=DotProduct(A=@n23, B=@n1), Exp=ScalarParameter() {defaultValue=53.7186, parameterName=Shine Smoothnes})
   B=Power(Base=DotProduct(A=@n2, B=@n23), Exp=ScalarParameter() {defaultValue=9.678, parameterName=Param_2})
   Alpha=NamedRerouteUsage()
  )
  n25 = SkyLightEnvMapSample(Direction=ReflectionVectorWS())
  n26 = LinearInterpolate(B=Multiply(A=Multiply(A=@n24, B=@n25), B=ScalarParameter() {defaultValue=0.5, parameterName=Fake Spec Intensity}), Alpha=@n21)
  n27 = LinearInterpolate(
   B=ScalarParameter() {group=Total Flock - Pickaxe, parameterName=Reactivity}
   Alpha=ScalarParameter() {defaultValue=1, group=Total Flock - Pickaxe, parameterName=Enable Reactivity}
  )
  MaterialFunctionCall(
   Material=BlendMaterialAttributes(
    A=@n4.MaterialAttributes
    B=SetMaterialAttributes!(MaterialAttributes=@n4.MaterialAttributes, Base Color=@n22, Emissive Color=@n26) {attributeSetTypes [#1, #2]}
    Alpha=Multiply(A=@n21, B=@n27)
   )
   M=@n3.M
   Pre Skinned Local Position=@n3.M
   Distance Blend=@n3.M
   Fresnel=@n3.M
  ) {materialFunction=&MF_CharacterParent_2023_PostFX}
  n28 = Subtract(A=Time(), B=ScalarParameter() {defaultValue=-100, group=Total Flock - Pickaxe, parameterName=DmgReactivityStartTime_1})
  n29 = Saturate(None=Divide(A=@n28, B=ScalarParameter() {defaultValue=0.4, group=Total Flock - Pickaxe, parameterName=ReactivityRampUpDuration}) {constB=0.8})
  n30 = ScalarParameter() {defaultValue=0.75, group=Total Flock - Pickaxe, parameterName=ReactivityRampUpExp}
  n31 = ScalarParameter() {defaultValue=10, group=Total Flock - Pickaxe, parameterName=ReactivityRampDownDuration}
  n32 = OneMinus(None=Saturate(None=Divide(A=@n28, B=@n31) {constB=0.8}))
  n33 = ScalarParameter() {defaultValue=1, group=Total Flock - Pickaxe, parameterName=ReactivityRampDownExp}
  n34 = Subtract(A=Time(), B=ScalarParameter() {defaultValue=-100, group=Total Flock - Pickaxe, parameterName=DmgReactivityStartTime_0})
  n35 = Saturate(None=Divide(A=@n34, B=ScalarParameter() {defaultValue=0.4, group=Total Flock - Pickaxe, parameterName=ReactivityRampUpDuration}) {constB=0.8})
  n36 = ScalarParameter() {defaultValue=0.75, group=Total Flock - Pickaxe, parameterName=ReactivityRampUpExp}
  n37 = ScalarParameter() {defaultValue=10, group=Total Flock - Pickaxe, parameterName=ReactivityRampDownDuration}
  n38 = OneMinus(None=Saturate(None=Divide(A=@n34, B=@n37) {constB=0.8}))
  n39 = ScalarParameter() {defaultValue=1, group=Total Flock - Pickaxe, parameterName=ReactivityRampDownExp}
  n40 = Add(
   A=Multiply(A=Power(Base=@n29, Exp=@n30) {constExponent=0.8}, B=Power(Base=@n32, Exp=@n33) {constExponent=0.4})
   B=Multiply(A=Power(Base=@n35, Exp=@n36) {constExponent=0.8}, B=Power(Base=@n38, Exp=@n39) {constExponent=0.4})
  )
  n41 = Multiply(A=@n40, B=ScalarParameter() {defaultValue=0.4, group=Total Flock - Pickaxe, parameterName=ReactivitySingleHitStrength})
  n42 = Multiply(A=Saturate(None=@n40), B=ScalarParameter() {defaultValue=0.7, group=Total Flock - Pickaxe, parameterName=ReactivityFirstHitStrengthBoost}) {constB=0.4}
  NamedRerouteDeclaration(None=LinearInterpolate(
    A=ScalarParameter() {defaultValue=1, group=Total Flock - Pickaxe, parameterName=Time Step}
    B=Add(A=@n41, B=@n42)
    Alpha=ScalarParameter() {defaultValue=1, group=Total Flock - Pickaxe, parameterName=Enable Reactivity}
   )) {name=Reactivity, nodeColor=<rgba 0 1 0.3203 1>}
  NamedRerouteUsage()
  OneMinus(None=Fresnel(ExponentIn=ScalarParameter() {defaultValue=0.0267, parameterName=Param_1}))
  Saturate(None=@n16)
  VectorParameter() {parameterName=Param_1, defaultValue=<rgba 0.2893 0.0061 0.9045 1>}
 ]
}
